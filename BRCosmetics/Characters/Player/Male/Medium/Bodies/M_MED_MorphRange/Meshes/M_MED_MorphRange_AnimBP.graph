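# Post-copy animation Blueprint for the male medium MorphRange body part, adding secondary motion to its accessories.
# AnimGraph copies the attached parent's pose through a Control Rig, a thigh_r-based RigidBody node, then chained spring AnimDynamics for back handkerchief, ring and right holster ribbons.
# The dyn_holster_r AnimDynamics node is unconnected.
# EventGraph updates GravityOverride from a parameter asset each frame, but no shown node reads it; AnimDynamics gravity is fixed.
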AnimBlueprint /BRCosmetics/Characters/Player/Male/Medium/Bodies/M_MED_MorphRange/Meshes/M_MED_MorphRange_AnimBP
target_skeleton=&/M_MED_MorphRange_Skeleton
graphs [
 AnimGraph {nodes [
   Anim Dynamics - Bone: dyn_holster_r(Alpha=1, ExternalForce="0,0,0") {
    angularDampingOverride=1
    angularSpringConstant=20000
    bAngularSpring=true
    bOverrideAngularDamping=true
    bUseGravityOverride=true
    gravityScale=0.35
    simulationSpace=RootRelative
    boundBone {boneName=dyn_holster_r}
    componentLinearAccScale=<1 1 1>
    componentLinearVelScale=<1 1 1>
    gravityOverride {z=-980}
    physicsBodyDefinitions [{
      boundBone {boneName=dyn_holster_r}
      boxExtents=<20 30 30>
      constraintSetup {angularLimitsMax=<15 15 15>, angularLimitsMin=<-15 -15 -15>, angularTarget=<1 0 0>}
      localJointOffset=<4 0 0>
     }]
   }
   n1 = Control Rig(Source=Copy Pose From Mesh() {bUseAttachedParent=true}) {controlRigAssetReference {controlRigAsset=&/M_MED_MorphRange_CtrlRig}}
   n2 = RigidBody(ComponentPose=Local To Component(LocalPose=@n1)) {
    baseBoneRef {boneName=thigh_r}
    simSpaceSettings {
     dampingAlpha=0
     maxAngularAcceleration=9
     maxAngularVelocity=4
     maxLinearAcceleration=1000
     maxLinearVelocity=400
     velocityScaleZ=0.75
     worldAlpha=1
     externalLinearDragV=<1.1 1.1 1.1>
    }
   }
   n3 = Anim Dynamics - Bone: dyn_Handkerchief_01_back(ComponentPose=@n2, Alpha=1, ExternalForce="0,0,0") {
    angularDampingOverride=1
    angularSpringConstant=20000
    bAngularSpring=true
    bOverrideAngularDamping=true
    bUseGravityOverride=true
    gravityScale=0.35
    simulationSpace=RootRelative
    boundBone {boneName=dyn_Handkerchief_01_back}
    componentLinearAccScale=<1 1 1>
    componentLinearVelScale=<1 1 1>
    gravityOverride {z=-980}
    physicsBodyDefinitions [{
      boundBone {boneName=dyn_Handkerchief_01_back}
      boxExtents=<20 30 30>
      constraintSetup {angularLimitsMax=<0 0 20>, angularLimitsMin=<0 0 -30>, angularTarget=<1 0 0>}
      localJointOffset=<4 0 0>
     }]
   }
   n4 = Anim Dynamics - Bone: dyn_Handkerchief_02_back(ComponentPose=@n3, Alpha=1, ExternalForce="0,0,0") {
    angularDampingOverride=1
    angularSpringConstant=20000
    bAngularSpring=true
    bOverrideAngularDamping=true
    bUseGravityOverride=true
    gravityScale=0.35
    simulationSpace=RootRelative
    boundBone {boneName=dyn_Handkerchief_02_back}
    componentLinearAccScale=<1 1 1>
    componentLinearVelScale=<1 1 1>
    gravityOverride {z=-980}
    physicsBodyDefinitions [{
      boundBone {boneName=dyn_Handkerchief_02_back}
      boxExtents=<20 30 30>
      constraintSetup {angularLimitsMax=<0 0 20>, angularLimitsMin=<0 0 -30>, angularTarget=<1 0 0>}
      localJointOffset=<4 0 0>
     }]
   }
   n5 = Anim Dynamics - Bone: dyn_ring_back(ComponentPose=@n4, Alpha=1, ExternalForce="0,0,0") {
    angularDampingOverride=1
    angularSpringConstant=20000
    bAngularSpring=true
    bOverrideAngularDamping=true
    bUseGravityOverride=true
    gravityScale=0.35
    simulationSpace=RootRelative
    boundBone {boneName=dyn_ring_back}
    componentLinearAccScale=<1 1 1>
    componentLinearVelScale=<1 1 1>
    gravityOverride {z=-980}
    physicsBodyDefinitions [{
      boundBone {boneName=dyn_ring_back}
      boxExtents=<20 30 30>
      constraintSetup {angularLimitsMax=<0 3 5>, angularLimitsMin=<0 -3 -30>, angularTarget=<1 0 0>}
      localJointOffset=<4 0 0>
     }]
   }
   n6 = Anim Dynamics - Bone: dyn_ribbon_holster_02_r(ComponentPose=@n5, Alpha=1, ExternalForce="0,0,0") {
    angularDampingOverride=1
    angularSpringConstant=20000
    bAngularSpring=true
    bOverrideAngularDamping=true
    bUseGravityOverride=true
    gravityScale=0.35
    simulationSpace=RootRelative
    boundBone {boneName=dyn_ribbon_holster_02_r}
    componentLinearAccScale=<1 1 1>
    componentLinearVelScale=<1 1 1>
    gravityOverride {z=-980}
    physicsBodyDefinitions [{
      boundBone {boneName=dyn_ribbon_holster_02_r}
      boxExtents=<20 30 30>
      constraintSetup {angularLimitsMax=<0 15 10>, angularLimitsMin=<0 10 -10>, angularTarget=<1 0 0>}
      localJointOffset=<4 0 0>
     }]
   }
   n7 = Anim Dynamics - Bone: dyn_ribbon_holster_01_r(ComponentPose=@n6, Alpha=1, ExternalForce="0,0,0") {
    angularDampingOverride=1
    angularSpringConstant=20000
    bAngularSpring=true
    bOverrideAngularDamping=true
    bUseGravityOverride=true
    gravityScale=0.35
    simulationSpace=RootRelative
    boundBone {boneName=dyn_ribbon_holster_01_r}
    componentLinearAccScale=<1 1 1>
    componentLinearVelScale=<1 1 1>
    gravityOverride {z=-980}
    physicsBodyDefinitions [{
      boundBone {boneName=dyn_ribbon_holster_01_r}
      boxExtents=<20 30 30>
      constraintSetup {angularLimitsMax=<0 15 10>, angularLimitsMin=<0 10 -10>, angularTarget=<1 0 0>}
      localJointOffset=<4 0 0>
     }]
   }
   Output Pose(Result=Component To Local(ComponentPose=@n7)) {layerGroup=None}
  ]}
 EventGraph {nodes [
   Event BlueprintUpdateAnimation()
   > n1 = "Gravity Override State Function (Parameter Asset) (Nativized)"(
    CharacterPartAnimInstance=Self-Reference()
    GameTimeInSeconds=GetGameTimeInSeconds()
    Parameters=&/M_MED_MorphRange_Parameters
   )
   > Set GravityOverride(GravityOverride=@n1)
  ]}
]
variables [GravityOverride {type=Vector}]
