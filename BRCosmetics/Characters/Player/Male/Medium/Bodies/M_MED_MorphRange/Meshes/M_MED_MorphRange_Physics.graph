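# Physics asset for the medium male MorphRange player body, covering only the pelvis, right thigh and right dynamic holster bone.
# Each bone has a single capsule body, forming a minimal setup around the holster rather than a full-body ragdoll.
# Constraints chain the holster to the thigh and the thigh to the pelvis, with the holster joint given tighter, asymmetric swing limits and a zero twist limit.
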PhysicsAsset /BRCosmetics/Characters/Player/Male/Medium/Bodies/M_MED_MorphRange/Meshes/M_MED_MorphRange_Physics
bodies {
 dyn_holster_r [{shapeType=Capsule, length=20, radius=5, center=<12 0 -5>, rotation=<rot 90 0 0>}]
 pelvis [{shapeType=Capsule, length=23.0221, radius=14.8786, center=<7.0233 -1.1121 0.1769>, rotation=<rot 6.4314 -90.0003 -92.0423>}]
 thigh_r [{shapeType=Capsule, length=20.6442, radius=5.9577, center=<22.4692 -0.1575 -1.3288>, rotation=<rot 0 -90.983 -89.6515>}]
}
constraints [
 {bone1Name=thigh_r, bone2Name=pelvis, swing1LimitDegrees=45, swing2LimitDegrees=45, twistLimitDegrees=45}
 {bone1Name=dyn_holster_r, bone2Name=thigh_r, swing1LimitDegrees=15, swing2LimitDegrees=35, twistLimitDegrees=0}
]
